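# Sound cue for the activation sound of Creative-mode trigger toys, part of the trigger and receiver set.
# A single wave player feeds the TriggerReceive_Activate01 wave directly to the output at reduced volume, with no randomization or modulation.
# It uses small-toy attenuation, the Creative gadgets sound class and a concurrency asset named to allow only one trigger/receiver sound at a time.
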SoundCue /Game/Sounds/Creative/Toys/TriggerandReceivers/Trigger_Activate_Cue
root=WavePlayer_1
nodes [{id=WavePlayer_1, properties {soundWaveAssetPtr=&/TriggerReceive_Activate01}}]
properties {
 attenuationSettings=/Game/Sounds/Creative/Attenuations/Fort_Creative_SmallToy_Att
 maxDistance=3000
 soundClassObject=&Fort_Gadgets_Creative
 volumeMultiplier=0.6
 concurrencySet [/Game/Sounds/Creative/Concurrency/TriggerReciever_OnlyOne_Conc]
}
