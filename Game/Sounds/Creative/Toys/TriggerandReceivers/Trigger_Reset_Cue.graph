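# Sound cue for the reset sound of Creative trigger and receiver toys, playing a single TriggerReceive_Reset_01 wave.
# One wave player is the root, with no randomization or modulation nodes.
# It plays at reduced volume in the Creative gadgets sound class, with small-toy attenuation and a concurrency set named to allow only one trigger/receiver sound at a time.
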SoundCue /Game/Sounds/Creative/Toys/TriggerandReceivers/Trigger_Reset_Cue
root=WavePlayer_0
nodes [{id=WavePlayer_0, properties {soundWaveAssetPtr=&/TriggerReceive_Reset_01}}]
properties {
 attenuationSettings=/Game/Sounds/Creative/Attenuations/Fort_Creative_SmallToy_Att
 maxDistance=3000
 priority=0.9
 soundClassObject=&Fort_Gadgets_Creative
 volumeMultiplier=0.3
 concurrencySet [/Game/Sounds/Creative/Concurrency/TriggerReciever_OnlyOne_Conc]
}
